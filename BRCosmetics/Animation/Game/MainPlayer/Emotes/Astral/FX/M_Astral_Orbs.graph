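# Unlit alpha-composite Niagara sprite material for the Astral emote's orbs, drawing a procedural eclipse.
# Eroded radial gradients shape the disc and sun, with a sine-wiggled rotation driving a varying falloff exponent.
# Bent polar UVs pan grayscale noise outward or inward over time, overlaid on the disc and hue-shifted by intensity.
# Depth fade blends in a second color near intersections, while particle alpha scales emissive and opacity.
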Material /BRCosmetics/Animation/Game/MainPlayer/Emotes/Astral/FX/M_Astral_Orbs
graph {
 properties {
  bAutomaticallySetUsageInEditor=false
  bUsedWithNiagaraSprites=true
  bUsedWithStaticMesh=false
  blendMode=BLEND_AlphaComposite
  pixelDepthOffsetMode=PDOM_Legacy
  shadingModel=MSM_Unlit
  usedShadingModels=Unlit
 }
 nodes [
  Constant() {r=1}
  n1 = ParticleColor()
  n2 = TextureCoordinate()
  n3 = Constant() {r=0.5}
  n4 = Saturate(None=Divide(A=Distance(A=@n2, B=@n3), B=@n3))
  n5 = Constant() {r=0.2}
  n6 = Multiply(A=OneMinus(None=MaterialFunctionCall!(Gradient=@n4, Erosion Amount=@n5) {materialFunction=/BRCosmetics/Athena/Effects/Particles/Skydiving/NS_Contrail_DarkAndLight/Functions/MF_ErosionNormalized}.Result))
  n7 = MaterialFunctionCall(Gradient=OneMinus(None=@n4), Erosion Amount=ScalarParameter() {defaultValue=0.5, parameterName=Sun Size}) {materialFunction=&MF_ErosionNormalized}
  n8 = Add(A=Multiply(A=Time(), B=ScalarParameter() {defaultValue=0.1, parameterName=Eclipse Wiggle Speed, sliderMax=1}), B=@n4)
  n9 = Rotator(
   Coordinate=@n2
   Time=Multiply(A=Sine(None=@n8), B=ScalarParameter() {defaultValue=0.05, parameterName=Eclipse Wiggle Range, sliderMax=1}) {constB=0.3}
  ) {speed=1}
  n10 = LinearInterpolate(
   A=ScalarParameter() {defaultValue=5, parameterName=Eclipse Power A, sliderMax=66}
   B=ScalarParameter() {defaultValue=5, parameterName=Eclipse Power B, sliderMax=66}
   Alpha=MaterialFunctionCall!(
    Gradient=MaterialFunctionCall!(UV=@n9) {materialFunction=/BRCosmetics/Athena/Effects/Particles/Skydiving/NS_Contrail_DarkAndLight/Functions/MF_GradientDiagonal}."0, 0, 1, 1"
    Center Point=Constant() {r=0.5}
   ) {materialFunction=/Game/Effects/Fort_Effects/MaterialFunctions/MF_GradientReflected}.Result
  )
  n11 = OneMinus(None=Power(Base=OneMinus(None=@n7.Result), Exp=@n10))
  n12 = VectorParameter() {parameterName="Depth Fade Color && Distance", defaultValue=<rgba 0.0707 0 0.3 33>}
  n13 = VectorParameter() {parameterName="Eclipse Color && Hue Shift", defaultValue=<rgba 0.012 0.5 0 0.016>}
  n14 = MaterialFunctionCall!(XY=Subtract(A=@n9) {constB=0.5}) {materialFunction=/BRCosmetics/Athena/Effects/Particles/Skydiving/NS_Contrail_DarkAndLight/Functions/MF_CartesianToPolar}
  n15 = Subtract(A=@n14.Angle) {constB=0.25}
  n16 = Multiply(
   A=Multiply(A=Sine(None=@n15), B=ScalarParameter() {defaultValue=0.5, parameterName=Bend, sliderMax=1})
   B=Multiply(A=Multiply(A=@n14.Radius, B=@n14.Radius), B=@n14.Radius)
  )
  n17 = MaterialFunctionCall() {materialFunction=/BRCosmetics/Athena/Effects/Particles/Skydiving/NS_Contrail_DarkAndLight/MF_DarkAndLight_TimeOffset}
  n18 = ScalarParameter() {defaultValue=0.5, parameterName=Noise Pan Speed}
  n19 = AppendVector(
   A=Add(A=DynamicParameter() {paramNames [TimeOffset, "", "", ""]}.TimeOffset, B=Subtract(A=@n14.Angle, B=@n16))
   B=Subtract(A=Add(A=@n14.Radius, B=@n14.Radius), B=Multiply(A=@n17.Time with Offset, B=@n18))
  )
  n20 = AppendVector(
   A=ScalarParameter() {defaultValue=2, parameterName=UV Scale U, sliderMax=2}
   B=ScalarParameter() {defaultValue=2, parameterName=UV Scale V, sliderMax=2}
  )
  n21 = TextureCoordinate()
  n22 = TextureSample!(UVs=Frac(None=Multiply(A=@n19, B=@n20) {constA=1}), Tex, "DDX(UVs)"=DDX(Value=@n21), "DDY(UVs)"=DDY(Value=@n21), Apply View MipBias) {
   mipValueMode=TMVM_Derivative
   samplerType=SAMPLERTYPE_LinearGrayscale
   texture=/BRCosmetics/Athena/Effects/Particles/Skydiving/NS_Contrail_DarkAndLight/T_DarkAndLight_Noise
  }
  n23 = MaterialFunctionCall!(Base=@n6, Blend=@n22.R) {materialFunction=/BRCosmetics/Athena/Effects/Particles/Skydiving/NS_Contrail_DarkAndLight/Functions/MF_BlendOverlay_Float}
  n24 = Saturate(None=Saturate(None=Power(Base=@n23.Result, Exp=@n10)))
  n25 = Multiply(A=@n13.A, B=@n24)
  n26 = Multiply(A=@n24, B=@n13.RGB)
  n27 = DepthFade(FadeDistance=@n12.A)
  n28 = Constant() {r=0.44}
  n29 = Multiply(
   A=Power(Base=@n7.Result, Exp=ScalarParameter() {defaultValue=1, parameterName=Eclipse Depth Fade Radius Exponent})
   B=Power(
    Base=MaterialFunctionCall(Gradient=@n27, Center Point=@n28) {materialFunction=&MF_GradientReflected}.Result
    Exp=ScalarParameter() {defaultValue=3, parameterName=Eclipse Depth Fade Exponent}
   )
  )
  n30 = AppendVector(
   A=LinearInterpolate(
    A=Multiply(A=@n11, B=@n12.RGB)
    B=MaterialFunctionCall(Hue Shift Percentage=@n25, Texture=@n26) {materialFunction=&HueShift}.Result
    Alpha=OneMinus(None=@n29)
   )
   B=@n6
  ) {desc=Eclipse RGBA}
  MaterialFunctionCall(Emissive=Multiply(A=@n30, B=@n1.A) {constB=0}) {materialFunction=&MF_FX_EmissiveAdjustments}
  Multiply(A=Power(Base=ComponentMask(None=@n30) {a=true}, Exp=ScalarParameter() {defaultValue=5, parameterName=Alpha Exponent, sliderMax=22}), B=@n1.A) {constB=0}
  Subtract(B=Multiply(A=@n11) {constB=0})
 ]
}
